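; PCG biome generator graph that places points along terrain elevation contours inside a biome's bounding shape.
; Height limits clamped to terrain bounds drive either native Elevation Isolines with spline sampling (default) or a subgraph implementation, with spacing adjusted for compatibility.
; A loop filters points using slope-derived horizontal distances, then a subgraph extracts crest and toe points.
; Survivors are scaled by remapped Distance, randomly transformed and thinned by density noise, with optional debug display.
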(graph BiomeGenerator_ElevationLines
  (params (ElevationLines_DisplayAllPoints :type Bool) (ElevationLines_HeightMin :type Double :default 4e+03) (ElevationLines_HeigntMax :type Double :default 6e+03) (ElevationLines_VerticalSpacing :type Double :default 1e+02) (ElevationLines_PointWidth :type Double :default 1e+02) (ElevationLines_DisplayFilteredPoints :type Bool) (ElevationLines_IsolatedPointsFilterSearchDistance :type Double :default 5e+02) (ElevationLines_CornerPointsFilterSearchDistance :type Double :default 1e+02) (ElevationLines_SlopeRatioFilterRange :type (Struct /Script/Engine.InputRange) :default (struct :Min 0.5)) (ElevationLines_DisplaySlopeGradients :type Bool) (ElevationLines_SlopeFilterMode :type (Enum /PCG/Subgraphs/ElevationIsolineSlopeFilterMode.ElevationIsolineSlopeFilterMode)) (ElevationLines_DistanceToCrest :type Double) (ElevationLines_DistanceToToe :type Double) (ElevationLines_OrientPointsZUp :type Bool :default true) (ElevationLines_CollisionChannel :type (Enum /Script/Engine.ECollisionChannel)) (OffsetMin :type Vector) (OffsetMax :type Vector) (RotationMin :type Rotator) (RotationMax :type Rotator) (ScaleMin :type Double :default 1.0) (ScaleMax :type Double :default 1.0) (ScaleByDistanceMin :type Double :default 1.0) (ScaleByDistanceMax :type Double :default 1.0) (DensityFilter :type Double :default 1.0) (DensityNoiseSeed :type Int :default 42) (UseNativeIsolinesGenerator :type Bool :default true))
  (bind Branch_50 (Branch :bOutputToB-param UseNativeIsolinesGenerator))
  (bind Subgraph_2 (Subgraph :Graph /PCG/Subgraphs/ElevationIsolines.ElevationIsolines :CollisionChannel ElevationLines_CollisionChannel :DisplayPoints ElevationLines_DisplayAllPoints :SplineSamplingDistance ElevationLines_PointWidth :VerticalSpacing ElevationLines_VerticalSpacing :ZUp ElevationLines_OrientPointsZUp))
  (bind Loop_3 (Loop :bTokenizeOnWhiteSpace true :Graph /PCG/Subgraphs/ElevationIsolinesFilters.ElevationIsolinesFilters :CornerFilterSearchDistance ElevationLines_CornerPointsFilterSearchDistance :DisplayPoints ElevationLines_DisplayFilteredPoints :NeighborSearchDistance ElevationLines_IsolatedPointsFilterSearchDistance))
  (bind Merge_14 (Merge%20Points :bMergeMetadata false))
  (bind Subgraph_8 (Subgraph :Graph /PCG/Subgraphs/ElevationIsolinesExtractCrestAndToePoints.ElevationIsolinesExtractCrestAndToePoints :DisplaySlopeGradients ElevationLines_DisplaySlopeGradients :DistanceToCrest ElevationLines_DistanceToCrest :DistanceToToe ElevationLines_DistanceToToe :SlopeFilterMode ElevationLines_SlopeFilterMode :VerticalSpacing ElevationLines_VerticalSpacing))
  (bind PointToAttributeSet_0 (Point%20To%20Attribute%20Set))
  (bind AttributeVectorOp_0 (Attribute%20Vector%20Op :InputSource1 Distance :Operation Length))
  (bind AttributeReduce_0 (Attribute%20Reduce :Operation Min))
  (bind AttributeReduce_1 (Attribute%20Reduce :Operation Max))
  (bind AttributeRemap_0 (Attribute%20Remap :OutRangeMax-param ScaleByDistanceMax :OutRangeMin-param ScaleByDistanceMin))
  (bind DeleteAttributes_0 (Delete%20Attributes :SelectedAttributes "Distance"))
  (bind AttributeMathsOp_2 (Attribute%20Maths%20Op :InputSource1 $Scale :InputSource2 Distance :Operation Multiply))
  (bind TransformPoints_0 (Transform%20Points :OffsetMax (0.0 -5e+01 -8e+01) :OffsetMin (0.0 -5e+01 -8e+01) :OffsetMax-param OffsetMax :OffsetMin-param OffsetMin :RotationMax-param RotationMax :RotationMin-param RotationMin :ScaleMax-param ScaleMax :ScaleMin-param ScaleMin))
  (bind AttributeNoise_0 (Attribute%20Noise :Seed-param DensityNoiseSeed))
  (bind DensityFilter_1 (Density%20Filter :LowerBound 0.0 :UpperBound-param DensityFilter))
  (bind Branch_63 (Branch :bOutputToB-param ElevationLines_DisplayFilteredPoints))
  (bind AttributeBooleanOp_22 (Attribute%20Boolean%20Op :Operation Not :In ElevationLines_OrientPointsZUp))
  (bind ElevationIsolines_0 (Elevation%20Isolines :bLinearSpline true :bOutputAsSpline true :bProjectSurfaceNormal true))
  (bind Debug_0 (Debug))
  ;[0 0 0] Inputs
  (bind FilterDataByTag_1 (Filter%20Data%20By%20Tag :bTokenizeOnWhiteSpace true :SelectedTags "BoundingShape"))
  (bind SurfaceSampler_0 (Surface%20Sampler :Looseness 0.0 :PointExtents (0.5 0.5 0.5) :PointsPerSquaredMeter 0.01))
  (bind GetBounds_15 (Get%20Bounds))
  (bind DeleteAttributes_18 (Delete%20Attributes :bTokenizeOnWhiteSpace true :Operation KeepSelectedAttributes :SelectedAttributes "BoundsMin"))
  (bind DeleteAttributes_20 (Delete%20Attributes :bTokenizeOnWhiteSpace true :Operation KeepSelectedAttributes :SelectedAttributes "BoundsMax"))
  (bind AttributeMathsOp_31 (Attribute%20Maths%20Op :InputSource1 @Last.z :Operation Max :OutputTarget HeightMin :InB ElevationLines_HeightMin))
  (bind AttributeMathsOp_32 (Attribute%20Maths%20Op :InputSource1 @Last.z :Operation Min :OutputTarget HeightMax :InB ElevationLines_HeigntMax))
  (bind DeleteAttributes_2 (Delete%20Attributes :Operation KeepSelectedAttributes :SelectedAttributes "HeightMin"))
  (bind DeleteAttributes_4 (Delete%20Attributes :Operation KeepSelectedAttributes :SelectedAttributes "HeightMax"))
  (bind Proxy_1 (Proxy :Settings-param DefaultTerrainGetterSetting))
  (bind FilterDataByType_1 (Filter%20Data%20By%20Type :TargetType (struct :Ids [(struct :Struct /Script/PCG.PCGDataTypeInfoSurface)])))
  (bind GetActorData_0 (Get%20Actor%20Data))
  (bind Intersection_1 (Intersection))
  ;[0 0 0] Compute Horizontal Distance From Vertical spacing and slope filter ratio
  (bind CreateAttribute_1 (Create%20Constant :AttributeTypes (struct :DoubleValue 1e-06)))
  (bind DeleteAttributes_52 (Delete%20Attributes :bTokenizeOnWhiteSpace true :Operation KeepSelectedAttributes :SelectedAttributes "Min" :In ElevationLines_SlopeRatioFilterRange))
  (bind AttributeMathsOp_0 (Attribute%20Maths%20Op :InputSource1 Min :Operation Max))
  (bind AttributeMathsOp_62 (Attribute%20Maths%20Op :Operation Divide :InA ElevationLines_VerticalSpacing))
  (bind CreateAttribute_28 (Create%20Constant :AttributeTypes (struct :DoubleValue 1e-06)))
  (bind DeleteAttributes_50 (Delete%20Attributes :bTokenizeOnWhiteSpace true :Operation KeepSelectedAttributes :SelectedAttributes "Max" :In ElevationLines_SlopeRatioFilterRange))
  (bind AttributeMathsOp_26 (Attribute%20Maths%20Op :InputSource1 Max :Operation Max))
  (bind AttributeMathsOp_24 (Attribute%20Maths%20Op :Operation Divide :InA ElevationLines_VerticalSpacing))
  (bind AttributeMathsOp_39 (Attribute%20Maths%20Op :Operation Subtract :InB ElevationLines_VerticalSpacing))
  (bind AttributeMathsOp_43 (Attribute%20Maths%20Op :Operation Subtract :InB ElevationLines_VerticalSpacing))
  (bind AttributeRename_55 (Attribute%20Rename :AttributeToRename ElevationLines_VerticalSpacing :NewAttributeName MaxHorizontalDistance))
  (bind AttributeRename_57 (Attribute%20Rename :AttributeToRename ElevationLines_VerticalSpacing :NewAttributeName MinHorizontalDistance))
  ;[0 0 0] For Compatibility with SG Implementation
  (bind AttributeMathsOp_8 (Attribute%20Maths%20Op :Operation Subtract))
  (bind AttributeMathsOp_9 (Attribute%20Maths%20Op :Operation Divide :InB ElevationLines_VerticalSpacing))
  (bind AttributeCast_0 (Attribute%20Cast :OutputType Integer32))
  (bind CreateAttribute_2 (Create%20Constant :AttributeTypes (struct :DoubleValue 1.0)))
  (bind AttributeMathsOp_10 (Attribute%20Maths%20Op))
  (bind AttributeMathsOp_13 (Attribute%20Maths%20Op :Operation Divide))
  ;[0 0 0] Prepare Points from Native Isolines
  (bind SplineSampler_14 (Spline%20Sampler :SamplerParams (struct :Mode Distance :DistanceIncrement 2.5e+02) :DistanceIncrement ElevationLines_PointWidth))
  (bind CreateAttribute_55 (Create%20Constant :AttributeTypes (struct :DoubleValue 1.0)))
  (bind CreateAttribute_61 (Create%20Constant :AttributeTypes (struct :DoubleValue 0.5)))
  (bind AttributeMathsOp_59 (Attribute%20Maths%20Op :Operation Multiply :InA ElevationLines_PointWidth))
  (bind CreateAttribute_57 (Create%20Constant :AttributeTypes (struct :DoubleValue 1e+01)))
  (bind MakeVectorAttribute_56 (Make%20Vector%20Attribute :OutputType Vector))
  (bind MergePoints_29 (Merge%20Points))
  (bind ExtentsModifier_66 (Extents%20Modifier :Extents (1.0 25.0 25.0)))
  (bind AttributeMathsOp_54 (Attribute%20Maths%20Op :InputSource1 $Density :Operation Set))
  (bind AttributeMathsOp_1 (Attribute%20Maths%20Op :InputSource1 $Steepness :Operation Set))
  (bind AttributePartition_20 (Attribute%20Partition :PartitionAttributeSelectors [$Position.z]))
  (bind Branch_64 (Branch :bOutputToB-param ElevationLines_DisplayAllPoints))
  (bind ExtentsModifier_65 (Extents%20Modifier :Extents (-1.0 -1.0 -1.0) :Mode Add))
  (bind Debug_62 (Debug))
  (reroute BoundingShape Intersection_1)
  (reroute HeightMax DeleteAttributes_4)
  (reroute HeightMin DeleteAttributes_2)
  (reroute Isolines%20Min%20Horizontal%20Distance AttributeRename_57)
  (reroute IsolinesMaxHorizontalDistance AttributeRename_55)
  (reroute Native%20Isoline%20Points AttributePartition_20)
  (reroute Native%20Isoline%20Splines ElevationIsolines_0)
  (reroute TerrainSurface FilterDataByType_1.InsideFilter)
  (connect AttributeMathsOp_9 Out AttributeCast_0 In)
  (connect DeleteAttributes_52 Out AttributeMathsOp_0 InA)
  (connect CreateAttribute_1 Out AttributeMathsOp_0 InB)
  (connect AttributeMathsOp_54 Out AttributeMathsOp_1 InA)
  (connect CreateAttribute_55 Out AttributeMathsOp_1 InB)
  (connect DeleteAttributes_0 Out AttributeMathsOp_2 InA)
  (connect AttributeRemap_0 Out AttributeMathsOp_2 InB)
  (connect HeightMax Out AttributeMathsOp_8 InA)
  (connect HeightMin Out AttributeMathsOp_8 InB)
  (connect AttributeMathsOp_8 Out AttributeMathsOp_9 InA)
  (connect AttributeCast_0 Out AttributeMathsOp_10 InA)
  (connect CreateAttribute_2 Out AttributeMathsOp_10 InB)
  (connect AttributeMathsOp_8 Out AttributeMathsOp_13 InA)
  (connect AttributeMathsOp_10 Out AttributeMathsOp_13 InB)
  (connect AttributeMathsOp_26 Out AttributeMathsOp_24 InB)
  (connect DeleteAttributes_50 Out AttributeMathsOp_26 InA)
  (connect CreateAttribute_28 Out AttributeMathsOp_26 InB)
  (connect DeleteAttributes_18 Out AttributeMathsOp_31 InA)
  (connect DeleteAttributes_20 Out AttributeMathsOp_32 InA)
  (connect AttributeMathsOp_62 Out AttributeMathsOp_39 InA)
  (connect AttributeMathsOp_24 Out AttributeMathsOp_43 InA)
  (connect ExtentsModifier_66 Out AttributeMathsOp_54 InA)
  (connect CreateAttribute_55 Out AttributeMathsOp_54 InB)
  (connect CreateAttribute_61 Out AttributeMathsOp_59 InB)
  (connect AttributeMathsOp_0 Out AttributeMathsOp_62 InB)
  (connect TransformPoints_0 Out AttributeNoise_0 In)
  (connect AttributeMathsOp_1 Out AttributePartition_20 In)
  (connect AttributeVectorOp_0 Out AttributeReduce_0 In)
  (connect AttributeVectorOp_0 Out AttributeReduce_1 In)
  (connect AttributeVectorOp_0 Out AttributeRemap_0 In)
  (connect AttributeReduce_1 Out AttributeRemap_0 InRangeMax)
  (connect AttributeReduce_0 Out AttributeRemap_0 InRangeMin)
  (connect AttributeMathsOp_39 Out AttributeRename_55 In)
  (connect AttributeMathsOp_43 Out AttributeRename_57 In)
  (connect PointToAttributeSet_0 Out AttributeVectorOp_0 In)
  (connect TerrainSurface Out Branch_50 In)
  (connect DensityFilter_1 Out Branch_63 In)
  (connect AttributePartition_20 Out Branch_64 In)
  (connect Native%20Isoline%20Points Out Debug_0 In)
  (connect Subgraph_2 Points Debug_0 In)
  (connect ExtentsModifier_65 Out Debug_62 In)
  (connect Subgraph_8 Out DeleteAttributes_0 In)
  (connect AttributeMathsOp_31 Out DeleteAttributes_2 In)
  (connect AttributeMathsOp_32 Out DeleteAttributes_4 In)
  (connect GetBounds_15 Out DeleteAttributes_18 In)
  (connect GetBounds_15 Out DeleteAttributes_20 In)
  (connect AttributeNoise_0 Out DensityFilter_1 In)
  (connect BoundingShape Out ElevationIsolines_0 BoundingShape)
  (connect AttributeBooleanOp_22 Out ElevationIsolines_0 bProjectSurfaceNormal)
  (connect HeightMax Out ElevationIsolines_0 ElevationEnd)
  (connect AttributeMathsOp_13 Out ElevationIsolines_0 ElevationIncrement)
  (connect HeightMin Out ElevationIsolines_0 ElevationStart)
  (connect Branch_50 Output%20B ElevationIsolines_0 In)
  (connect Branch_64 Output%20B ExtentsModifier_65 In)
  (connect MakeVectorAttribute_56 Out ExtentsModifier_66 Extents)
  (connect MergePoints_29 Out ExtentsModifier_66 In)
  (connect DefaultInputNode In FilterDataByTag_1 In)
  (connect Proxy_1 Out FilterDataByType_1 In)
  (connect SurfaceSampler_0 Out GetBounds_15 In)
  (connect FilterDataByTag_1 InsideFilter Intersection_1 Primary%20Source)
  (connect GetActorData_0 Out Intersection_1 Source1)
  (connect Native%20Isoline%20Points Out Loop_3 AllPointData)
  (connect Subgraph_2 Points Loop_3 AllPointData)
  (connect IsolinesMaxHorizontalDistance Out Loop_3 MaxHorizontalDistance)
  (connect Isolines%20Min%20Horizontal%20Distance Out Loop_3 MinHorizontalDistance)
  (connect Native%20Isoline%20Points Out Loop_3 PointDataLoop)
  (connect Subgraph_2 Points Loop_3 PointDataLoop)
  (connect AttributeMathsOp_59 Out MakeVectorAttribute_56 X)
  (connect CreateAttribute_57 Out MakeVectorAttribute_56 Y)
  (connect CreateAttribute_57 Out MakeVectorAttribute_56 Z)
  (connect Loop_3 Out Merge_14 In)
  (connect SplineSampler_14 Out MergePoints_29 In)
  (connect Subgraph_8 Out PointToAttributeSet_0 In)
  (connect BoundingShape Out SplineSampler_14 Bounding%20Shape)
  (connect Native%20Isoline%20Splines Out SplineSampler_14 Spline)
  (connect BoundingShape Out Subgraph_2 BoundingShape)
  (connect HeightMax Out Subgraph_2 HeightMax)
  (connect HeightMin Out Subgraph_2 HeightMin)
  (connect Branch_50 Output%20A Subgraph_2 Surface)
  (connect Merge_14 Out Subgraph_8 ElevationLinesPointsIn)
  (connect IsolinesMaxHorizontalDistance Out Subgraph_8 MaxHorizontalDistance)
  (connect TerrainSurface Out SurfaceSampler_0 Surface)
  (connect AttributeMathsOp_2 Out TransformPoints_0 In)
  (output Out Branch_63.Output%20A))
