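; PCG subgraph for biome generators that splits incoming data in two and labels each side with a generator sub-type.
; An attribute filter compares a parameterized target attribute against a threshold, with the operator also exposed as a graph parameter.
; The inside and outside results each receive a GeneratorSubType name attribute from their own parameter, then leave on separate point outputs.
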(graph BiomeGenerator_SubGeneratorSetup
  (params (AttributeFilterTarget :type (Struct /Script/PCG.PCGAttributePropertySelector) :default @Last) (AttributeFilterOperator :type (Enum /Script/PCG.EPCGAttributeFilterOperator) :default GreaterOrEqual) (AttributeFilterThresholdValue :type Double :default 0.5) (InsideFilter_GeneratorSubTypeName :type Name) (OutsideFilter_GeneratorSubTypeName :type Name))
  (outputs (InsideFilterGeneratorSubTypeData :type Point :advanced) (OutsideFilterGeneratorSubTypeData :type Point))
  (bind AttributeFilter_1 (Filter%20Attribute%20Elements :Operator GreaterOrEqual :TargetAttribute None :ThresholdAttribute AttributeFilterThresholdValue :Filter AttributeFilterThresholdValue :Operator-param AttributeFilterOperator :TargetAttribute-param AttributeFilterTarget))
  (bind AddAttribute_2 (Add%20Attribute :AttributeTypes (struct :Type Name) :OutputTarget GeneratorSubType :Attributes InsideFilter_GeneratorSubTypeName))
  (bind AddAttribute_3 (Add%20Attribute :AttributeTypes (struct :Type Name) :OutputTarget GeneratorSubType :Attributes OutsideFilter_GeneratorSubTypeName))
  (connect AttributeFilter_1 InsideFilter AddAttribute_2 In)
  (connect AttributeFilter_1 OutsideFilter AddAttribute_3 In)
  (connect DefaultInputNode In AttributeFilter_1 In)
  (output InsideFilterGeneratorSubTypeData AddAttribute_2)
  (output OutsideFilterGeneratorSubTypeData AddAttribute_3))
